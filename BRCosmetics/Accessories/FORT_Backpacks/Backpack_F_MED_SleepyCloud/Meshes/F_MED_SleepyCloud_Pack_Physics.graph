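# Physics asset for the Sleepy Cloud backpack cosmetic, defining a jointed rig with pelvis, spine, arms, legs, two peels and a dart.
# Small uniform capsules sit on base_ anchor bones, while mostly larger, fitted capsules cover the dyn_ bones.
# Angle-limited constraints chain dyn_ bones from separate dart, pelvis and spine anchors.
# Limb and peel anchors join the dyn_ torso with free swing and twist, then start their own limited chains.
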PhysicsAsset /BRCosmetics/Accessories/FORT_Backpacks/Backpack_F_MED_SleepyCloud/Meshes/F_MED_SleepyCloud_Pack_Physics
bodies {
 base_dart [{shapeType=Capsule, length=0.505, radius=0.505, shapeName=base_dart_capsule, rotation=<rot 0 -90 0>}]
 base_leg_l [{shapeType=Capsule, length=0.505, radius=0.505, shapeName=base_leg_l_capsule, rotation=<rot 0 -90 0>}]
 base_leg_r [{shapeType=Capsule, length=0.505, radius=0.505, shapeName=base_leg_r_capsule, rotation=<rot 0 -90 0>}]
 base_peel_l [{shapeType=Capsule, length=0.505, radius=0.505, shapeName=base_peel_l_capsule, rotation=<rot 0 -90 0>}]
 base_peel_r [{shapeType=Capsule, length=0.505, radius=0.505, shapeName=base_peel_r_capsule, rotation=<rot 0 -90 0>}]
 base_pelvis [{shapeType=Capsule, length=0.505, radius=0.505, shapeName=base_pelvis_capsule, rotation=<rot 0 -90 0>}]
 base_shoulder_l [{shapeType=Capsule, length=0.505, radius=0.505, shapeName=base_shoulder_l_capsule, rotation=<rot 0 -90 0>}]
 base_shoulder_r [{shapeType=Capsule, length=0.505, radius=0.505, shapeName=base_shoulder_r_capsule, rotation=<rot 0 -90 0>}]
 base_spine_01 [{shapeType=Capsule, length=0.505, radius=0.505, shapeName=base_spine_01_capsule, rotation=<rot 0 -90 0>}]
 dyn_dart_01 [{shapeType=Capsule, length=1.7994, radius=1.0871, shapeName=dyn_dart_01_capsule, center=<2.31 0 0>, rotation=<rot -90 0 0>}]
 dyn_dart_02 [{shapeType=Capsule, length=0.505, radius=0.505, shapeName=dyn_dart_02_capsule, center=<1 0 0>, rotation=<rot 0 0 90>}]
 dyn_elbow_l [{
   shapeType=Capsule
   length=1.7745
   radius=1.8801
   shapeName=dyn_elbow_l_capsule
   center=<1.9562 -0.1391 0.2948>
   rotation=<rot -76.7154 0 0.0001>
  }]
 dyn_elbow_r [{
   shapeType=Capsule
   length=1.4312
   radius=1.5912
   shapeName=dyn_elbow_r_capsule
   center=<1.9438 -0.0129 -0.1874>
   rotation=<rot -83.1787 -179.9999 -173.0955>
  }]
 dyn_knee_l [{
   shapeType=Capsule
   length=1.1372
   radius=2.5222
   shapeName=dyn_knee_l_capsule
   center=<2.8037 -0.2079 -0.0506>
   rotation=<rot 3.8626 -10.8405 -95.0301>
  }]
 dyn_knee_r [{
   shapeType=Capsule
   length=0.9938
   radius=2.397
   shapeName=dyn_knee_r_capsule
   center=<2.998 -0.2132 -0.3248>
   rotation=<rot 0 172.9479 -72.238>
  }]
 dyn_leg_l [{shapeType=Capsule, length=1.494, radius=1.9818, shapeName=dyn_leg_l_capsule, center=<1.941 0 0>, rotation=<rot -90 0 0>}]
 dyn_leg_r [{
   shapeType=Capsule
   length=1.9623
   radius=2.0195
   shapeName=dyn_leg_r_capsule
   center=<2.2752 0.4934 -0.2642>
   rotation=<rot -90 180 -177.831>
  }]
 dyn_peel_l_01 [{
   shapeType=Capsule
   length=1.2728
   radius=1.4403
   shapeName=dyn_peel_l_01_capsule
   center=<1.4382 -0.1183 -0.1532>
   rotation=<rot -1.0572 -90 -1.1667>
  }]
 dyn_peel_l_02 [{
   shapeType=Capsule
   length=0.7288
   radius=1.2734
   shapeName=dyn_peel_l_02_capsule
   center=<0.8861 1.0774 -0.0709>
   rotation=<rot 2.2017 -117.2404 3.5941>
  }]
 dyn_peel_r_01 [{shapeType=Capsule, length=2.9167, radius=0.5144, shapeName=dyn_peel_r_01_capsule, center=<1.1273 -0.0406 -0.0462>}]
 dyn_peel_r_02 [{shapeType=Capsule, length=2.1649, radius=0.5349, shapeName=dyn_peel_r_02_capsule, center=<0.7124 0.1754 0.1312>}]
 dyn_pelvis [{shapeType=Capsule, length=1.0966, radius=4.9096, shapeName=dyn_pelvis_capsule, center=<0 -0.557 0>}]
 dyn_shoulder_l [{shapeType=Capsule, length=0.8289, radius=1.2792, shapeName=dyn_shoulder_l_capsule, center=<0.8577 0 0>, rotation=<rot -90 180 180>}]
 dyn_shoulder_r [{
   shapeType=Capsule
   length=1.439
   radius=1.3336
   shapeName=dyn_shoulder_r_capsule
   center=<1.4439 0.0131 0.1409>
   rotation=<rot 5.2965 -90 -95.3133>
  }]
 dyn_spine_01 [{shapeType=Capsule, length=3.0524, radius=4.7488, shapeName=dyn_spine_01_capsule, center=<2.8253 0 0>}]
 dyn_spine_02 [{shapeType=Capsule, length=2.5253, radius=3.8177, shapeName=dyn_spine_02_capsule, center=<3.9549 0 0>}]
 dyn_spine_03 [{shapeType=Capsule, length=0.113, radius=2.8755, shapeName=dyn_spine_03_capsule, center=<3.4145 0 0>}]
}
constraints [
 {bone1Name=dyn_dart_01, bone2Name=base_dart, swing1LimitDegrees=45, swing2LimitDegrees=45, twistLimitDegrees=45}
 {bone1Name=dyn_dart_02, bone2Name=dyn_dart_01, swing1LimitDegrees=45, swing2LimitDegrees=45, twistLimitDegrees=45}
 {bone1Name=dyn_pelvis, bone2Name=base_pelvis, swing1LimitDegrees=45, swing2LimitDegrees=45, twistLimitDegrees=45}
 {
  bone1Name=base_leg_l
  bone2Name=dyn_pelvis
  swing1LimitDegrees=45
  swing1Motion=Free
  swing2LimitDegrees=45
  swing2Motion=Free
  twistLimitDegrees=45
  twistMotion=Free
 }
 {bone1Name=dyn_leg_l, bone2Name=base_leg_l, swing1LimitDegrees=45, swing2LimitDegrees=45, twistLimitDegrees=45}
 {bone1Name=dyn_knee_l, bone2Name=dyn_leg_l, swing1LimitDegrees=45, swing2LimitDegrees=45, twistLimitDegrees=45}
 {
  bone1Name=base_leg_r
  bone2Name=dyn_pelvis
  swing1LimitDegrees=45
  swing1Motion=Free
  swing2LimitDegrees=45
  swing2Motion=Free
  twistLimitDegrees=45
  twistMotion=Free
 }
 {bone1Name=dyn_leg_r, bone2Name=base_leg_r, swing1LimitDegrees=45, swing2LimitDegrees=45, twistLimitDegrees=45}
 {bone1Name=dyn_knee_r, bone2Name=dyn_leg_r, swing1LimitDegrees=45, swing2LimitDegrees=45, twistLimitDegrees=45}
 {bone1Name=dyn_spine_01, bone2Name=base_spine_01, swing1LimitDegrees=45, swing2LimitDegrees=45, twistLimitDegrees=45}
 {
  bone1Name=base_shoulder_l
  bone2Name=dyn_spine_01
  swing1LimitDegrees=45
  swing1Motion=Free
  swing2LimitDegrees=45
  swing2Motion=Free
  twistLimitDegrees=45
  twistMotion=Free
 }
 {bone1Name=dyn_shoulder_l, bone2Name=base_shoulder_l, swing1LimitDegrees=45, swing2LimitDegrees=45, twistLimitDegrees=45}
 {bone1Name=dyn_elbow_l, bone2Name=dyn_shoulder_l, swing1LimitDegrees=45, swing2LimitDegrees=45, twistLimitDegrees=45}
 {
  bone1Name=base_shoulder_r
  bone2Name=dyn_spine_01
  swing1LimitDegrees=45
  swing1Motion=Free
  swing2LimitDegrees=45
  swing2Motion=Free
  twistLimitDegrees=45
  twistMotion=Free
 }
 {bone1Name=dyn_shoulder_r, bone2Name=base_shoulder_r, swing1LimitDegrees=45, swing2LimitDegrees=45, twistLimitDegrees=45}
 {bone1Name=dyn_elbow_r, bone2Name=dyn_shoulder_r, swing1LimitDegrees=45, swing2LimitDegrees=45, twistLimitDegrees=45}
 {bone1Name=dyn_spine_02, bone2Name=dyn_spine_01, swing1LimitDegrees=45, swing2LimitDegrees=45, twistLimitDegrees=45}
 {bone1Name=dyn_spine_03, bone2Name=dyn_spine_02, swing1LimitDegrees=45, swing2LimitDegrees=45, twistLimitDegrees=45}
 {
  bone1Name=base_peel_l
  bone2Name=dyn_spine_03
  swing1LimitDegrees=45
  swing1Motion=Free
  swing2LimitDegrees=45
  swing2Motion=Free
  twistLimitDegrees=45
  twistMotion=Free
 }
 {bone1Name=dyn_peel_l_01, bone2Name=base_peel_l, swing1LimitDegrees=45, swing2LimitDegrees=45, twistLimitDegrees=45}
 {bone1Name=dyn_peel_l_02, bone2Name=dyn_peel_l_01, swing1LimitDegrees=45, swing2LimitDegrees=45, twistLimitDegrees=45}
 {
  bone1Name=base_peel_r
  bone2Name=dyn_spine_03
  swing1LimitDegrees=45
  swing1Motion=Free
  swing2LimitDegrees=45
  swing2Motion=Free
  twistLimitDegrees=45
  twistMotion=Free
 }
 {bone1Name=dyn_peel_r_01, bone2Name=base_peel_r, swing1LimitDegrees=45, swing2LimitDegrees=45, twistLimitDegrees=45}
 {bone1Name=dyn_peel_r_02, bone2Name=dyn_peel_r_01, swing1LimitDegrees=45, swing2LimitDegrees=45, twistLimitDegrees=45}
]
